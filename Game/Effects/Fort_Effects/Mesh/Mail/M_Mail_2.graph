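# M_Mail_2 is a two-sided Default Lit material for mail mesh particles in Fort effects.
# It samples the T_Mail texture and passes its RGB into ConvertFromDiffSpec as diffuse color.
# The graph also holds a Constant of 1, with no connections shown from either node to the material outputs.
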Material /Game/Effects/Fort_Effects/Mesh/Mail/M_Mail_2
graph {
 properties {bUsedWithMeshParticles=true, pixelDepthOffsetMode=PDOM_Legacy, twoSided=true, usedShadingModels=Default Lit}
 nodes [
  Constant() {r=1}
  MaterialFunctionCall(DiffuseColor=TextureSample() {texture=/Game/Effects/Fort_Effects/Textures/T_Mail}.RGB) {materialFunction=&ConvertFromDiffSpec}
 ]
}
