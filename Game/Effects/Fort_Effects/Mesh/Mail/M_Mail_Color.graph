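# M_Mail_Color is a two-sided, default lit effects material for mail mesh particles, tinting the T_Mail texture by particle color.
# The texture RGB is added to Particle Color, rather than multiplied, and feeds the DiffuseColor input of ConvertFromDiffSpec.
# A parallel Multiply of the same inputs and a Constant of 1 have no visible consumers, and no material output links appear.
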Material /Game/Effects/Fort_Effects/Mesh/Mail/M_Mail_Color
graph {
 properties {bUsedWithMeshParticles=true, pixelDepthOffsetMode=PDOM_Legacy, twoSided=true, usedShadingModels=Default Lit}
 nodes [
  Constant() {r=1}
  n1 = TextureSample() {texture=/Game/Effects/Fort_Effects/Textures/T_Mail}
  n2 = ParticleColor()
  MaterialFunctionCall(DiffuseColor=Add(A=@n1.RGB, B=@n2.RGB)) {materialFunction=&ConvertFromDiffSpec}
  Multiply(A=@n1.RGB, B=@n2.RGB)
 ]
}
